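# Master material for the liquid parts of the Alien Agent player skin, between the shared character PreFX and PostFX functions.
# Black base color leaves the look to emissive: view-dependent color blends, fresnel rim and outline, sphere-mask glow, and fake sun specular.
# A fresnel fade, signed by face direction, plus the specular forms the opacity mask, forced solid in shadow passes.
# The glow's bounds normalization subtracts Max from itself, dividing by zero.
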Material /Game/Characters/Player/Male/Medium/Bodies/M_MED_Alien_Agent/Materials/M_MED_AlienAgent_Liquid_Master
graph {
 properties {
  bAutomaticallySetUsageInEditor=false
  bDisableDepthTest=true
  bEnableResponsiveAA=true
  bUseMaterialAttributes=true
  bUsedWithMeshDeformer=true
  bUsedWithMorphTargets=true
  bUsedWithSkeletalMesh=true
  ditherOpacityMask=true
  pixelDepthOffsetMode=PDOM_Legacy
  shadingModel=MSM_FromMaterialExpression
  subsurfaceProfile=&SS_HeroSkin_02
  translucencyDirectionalLightingIntensity=5
  translucencyLightingMode=TLM_Surface
  usedShadingModels=Default Lit | Subsurface Profile
 }
 nodes [
  CollectionParameter() {collection=&FortniteMaterialParameters, parameterName=SunLightColor}
  n1 = MaterialFunctionCall() {materialFunction=&BaseFN_CharacterFunction_PreFX}
  n2 = GetMaterialAttributes(None=@n1.Result) {attributeGetTypes [#1]}
  n3 = ScalarParameter() {defaultValue=1, group=0SlurpMonsterSummer, parameterName=Fade_Min}
  n4 = ScalarParameter() {defaultValue=1, group=0SlurpMonsterSummer, parameterName=Fade_Max}
  n5 = Power(Base=@n1.Fresnel, Exp=ScalarParameter() {defaultValue=1, group=0SlurpMonsterSummer, parameterName=Fade_Power})
  n6 = Multiply(A=TwoSidedSign(), B=ScalarParameter() {defaultValue=-1, group=0SlurpMonsterSummer, parameterName=FlipNormals})
  n7 = CollectionParameter() {
   collection=&FortniteMaterialParameters
   parameterName=SunlightDirectionalVector
  }
  n8 = ComponentMask(None=@n7) {b=true, g=true, r=true}
  n9 = PixelNormalWS()
  n10 = Multiply(A=Multiply(A=DotProduct(A=@n8, B=@n9)) {constB=2}, B=@n9)
  n11 = DotProduct(
   A=Normalize(VectorInput=Subtract(A=@n10, B=@n8))
   B=Multiply(A=MaterialFunctionCall() {materialFunction=&CameraDirectionVector}.Result) {constB=-1}
  )
  n12 = Multiply(
   A=OneMinus(None=@n6)
   B=Power(Base=Max(A=@n11) {constB=0}, Exp=ScalarParameter() {defaultValue=15, group=0SlurpMonsterSummer, parameterName=FakeSpecPow})
  )
  n13 = Multiply(A=@n12, B=ScalarParameter() {defaultValue=1, group=0SlurpMonsterSummer, parameterName=FakeSpecOpacity}) {constB=0.4}
  n14 = Add(A=Multiply(A=LinearInterpolate(A=@n3, B=@n4, Alpha=@n5), B=@n6), B=@n13)
  n15 = VectorParameter() {group=0SlurpMonsterSummer, parameterName=FresnelColor, defaultValue=<rgba 0.0654 0.2865 0 0>}
  n16 = PreSkinnedLocalBounds()
  n17 = Divide(A=VertexInterpolator(VS=PreSkinnedPosition()).PS, B=Subtract(A=@n16.Max, B=@n16.Max))
  n18 = Multiply(A=SphereMask(A=ComponentMask(None=@n17) {g=true, r=true}, B=Constant2Vector(), Radius=Constant() {r=0.44}, Hardness=Constant() {r=0.1})) {constB=3}
  n19 = VectorParameter() {group=0SlurpMonsterSummer, parameterName=BaseColor, defaultValue=<rgba 0.012 0.1354 0.0182 0>}
  n20 = VectorParameter() {group=0SlurpMonsterSummer, parameterName=SubColor, defaultValue=<rgba 0.2655 0.1643 1 1>}
  n21 = Transform(None=Constant3Vector() {constant=<rgba 0 0 1 0>}) {transformSourceType=TRANSFORMSOURCE_Camera, transformType=TRANSFORM_Local}
  n22 = ComponentMask(None=@n21) {b=true}
  n23 = OneMinus(None=Saturate(None=Abs(None=@n22)))
  n24 = PixelNormalWS()
  n25 = Saturate(None=Abs(None=ComponentMask(None=@n21) {r=true}))
  n26 = Multiply(A=ScalarParameter() {defaultValue=1.5, group=0SlurpMonsterSummer, parameterName=FresnelPower}, B=Add(A=OneMinus(None=@n25)) {constB=0.01})
  n27 = ScalarParameter() {defaultValue=2.5, parameterName=Emissive_Outline}
  n28 = MaterialFunctionCall(Normal Vector=@n24, Power=ScalarParameter() {defaultValue=2.5, group=0SlurpMonsterSummer, parameterName=FresnelPower_Outline}) {materialFunction=&Fresnel_Function}
  n29 = Constant() {r=1}
  n30 = Add(
   A=LinearInterpolate(
    A=LinearInterpolate(A=@n19.RGB, B=@n20.RGB, Alpha=@n23)
    B=@n15.RGB
    Alpha=MaterialFunctionCall(Normal Vector=@n24, Power=@n26) {materialFunction=&Fresnel_Function}.Result
   )
   B=Multiply(
    A=Multiply(A=@n15.RGB, B=@n27)
    B=MaterialFunctionCall(In=@n28.Result, Contrast=@n29) {materialFunction=&CheapContrast}.Result
   )
  )
  n31 = Multiply(
   A=VectorParameter() {parameterName=FakeSpecColor, defaultValue=<rgba 1 1 1 0>}.RGB
   B=ScalarParameter() {defaultValue=2.45, group=0SlurpMonsterSummer, parameterName=FakeSpecIntensity}
  )
  n32 = Add(
   A=ComponentMask(None=Multiply(A=@n12, B=@n31)) {b=true, g=true, r=true}
   B=VectorParameter() {group=0SlurpMonsterSummer, parameterName=BaseEmissive}.RGB
  )
  n33 = SetMaterialAttributes!(
   MaterialAttributes=@n2.MaterialAttributes
   Base Color=Constant()
   Metallic=ScalarParameter() {group=0SlurpMonsterSummer, parameterName=Metallic}
   Specular=ScalarParameter() {group=0SlurpMonsterSummer, parameterName=Specular}
   Roughness=ScalarParameter() {defaultValue=0.32, group=0SlurpMonsterSummer, parameterName=Roughness}
   Opacity Mask=ShadowReplace(Default=Saturate(None=@n14), Shadow=Constant() {r=1})
   Normal=@n2.MaterialAttributes
   World Position Offset
   Emissive Color=Add(A=Multiply(A=@n18, B=@n15.RGB), B=Add(A=@n30, B=@n32))
  ) {attributeSetTypes [
    #2
    #3
    #4
    #5
    #6
    #1
    #7
    #8
   ]}
  MaterialFunctionCall(AO=@n1.AO, Distance Blend=@n1.AO, Fresnel=@n1.AO, Gmap=@n1.AO, Material=@n33) {materialFunction=&BaseFN_CharacterFunction_PostFX}
  ShadowReplace(Default=@n13, Shadow=Constant())
  VertexNormalWS()
 ]
}
