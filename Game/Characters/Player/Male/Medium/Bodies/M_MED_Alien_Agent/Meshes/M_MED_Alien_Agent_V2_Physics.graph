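# Physics asset for the M_MED_Alien_Agent_V2 player body, rigging its shirt and suit-back bones as jointed capsule chains.
# Eight chains each hang three segments from a small MSTR body constrained to the pelvis, mostly with free rotation.
# Segment joints use tight angular limits, except fully locked suit-back root joints.
# Pelvis and thigh capsules supply lower-body shapes; only the right thigh has a twist body.
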PhysicsAsset /Game/Characters/Player/Male/Medium/Bodies/M_MED_Alien_Agent/Meshes/M_MED_Alien_Agent_V2_Physics
bodies {
 dyn_shirt_01_l [{shapeType=Capsule, length=10, radius=5, shapeName=dyn_shirt_01_l_capsule, center=<1 0 4.6907>, rotation=<rot 90 -165.9638 104.0362>}]
 dyn_shirt_01_l1 [{shapeType=Capsule, length=10, radius=5, shapeName=dyn_shirt_01_l1_capsule, center=<1 0 4.6907>, rotation=<rot 90 -165.9638 104.0362>}]
 dyn_shirt_01_l2 [{shapeType=Capsule, length=10, radius=5, shapeName=dyn_shirt_01_l2_capsule, center=<1 0 4.6907>, rotation=<rot 90 -165.9638 104.0362>}]
 dyn_shirt_01_l_MSTR [{shapeType=Capsule, length=1, radius=1, shapeName=dyn_shirt_01_l_MSTR_capsule, rotation=<rot 0 0 90>}]
 dyn_shirt_01_r [{shapeType=Capsule, length=10, radius=5, shapeName=dyn_shirt_01_r_capsule, center=<1 0 4.6907>, rotation=<rot 90 -165.9638 104.0362>}]
 dyn_shirt_01_r1 [{shapeType=Capsule, length=10, radius=5, shapeName=dyn_shirt_01_r1_capsule, center=<1 0 4.6907>, rotation=<rot 90 -165.9638 104.0362>}]
 dyn_shirt_01_r2 [{shapeType=Capsule, length=10, radius=5, shapeName=dyn_shirt_01_r2_capsule, center=<1 0 4.6907>, rotation=<rot 90 -165.9638 104.0362>}]
 dyn_shirt_01_r_MSTR [{shapeType=Capsule, length=1, radius=1, shapeName=dyn_shirt_01_r_MSTR_capsule, rotation=<rot 0 0 90>}]
 dyn_shirt_fr_01_l [{shapeType=Capsule, length=10, radius=5, shapeName=dyn_shirt_fr_01_l_capsule, center=<1 0 3.6275>, rotation=<rot 90 180 90>}]
 dyn_shirt_fr_01_l1 [{shapeType=Capsule, length=10, radius=5, shapeName=dyn_shirt_fr_01_l1_capsule, center=<1 0 4.3242>, rotation=<rot 90 180 90>}]
 dyn_shirt_fr_01_l2 [{shapeType=Capsule, length=10, radius=5, shapeName=dyn_shirt_fr_01_l2_capsule, center=<1.6148 0 3.3603>, rotation=<rot 90 180 90>}]
 dyn_shirt_fr_01_l_MSTR [{shapeType=Capsule, length=1, radius=1, shapeName=dyn_shirt_fr_01_l_MSTR_capsule, rotation=<rot 0 0 90>}]
 dyn_shirt_fr_01_r [{shapeType=Capsule, length=10, radius=5, shapeName=dyn_shirt_fr_01_r_capsule, center=<1 0 4.2549>, rotation=<rot 90 180 90>}]
 dyn_shirt_fr_01_r1 [{shapeType=Capsule, length=10, radius=5, shapeName=dyn_shirt_fr_01_r1_capsule, center=<1 0 4.1422>, rotation=<rot 90 180 90>}]
 dyn_shirt_fr_01_r2 [{shapeType=Capsule, length=10, radius=5, shapeName=dyn_shirt_fr_01_r2_capsule, center=<2.0148 0 3.3495>, rotation=<rot 90 165.9638 75.9638>}]
 dyn_shirt_fr_01_r_MSTR [{shapeType=Capsule, length=1, radius=1, shapeName=dyn_shirt_fr_01_r_MSTR_capsule, rotation=<rot 0 0 90>}]
 dyn_shirt_fr_mid_01_l [{shapeType=Capsule, length=10, radius=5, shapeName=dyn_shirt_fr_mid_01_l_capsule, center=<1 0 4.6907>, rotation=<rot 90 -165.9638 104.0362>}]
 dyn_shirt_fr_mid_01_l1 [{shapeType=Capsule, length=10, radius=5, shapeName=dyn_shirt_fr_mid_01_l1_capsule, center=<1 0 4.6907>, rotation=<rot 90 -165.9638 104.0362>}]
 dyn_shirt_fr_mid_01_l2 [{shapeType=Capsule, length=10, radius=5, shapeName=dyn_shirt_fr_mid_01_l2_capsule, center=<1 0 4.6907>, rotation=<rot 90 -165.9638 104.0362>}]
 dyn_shirt_fr_mid_01_l_MSTR [{shapeType=Capsule, length=1, radius=1, shapeName=dyn_shirt_fr_mid_01_l_MSTR_capsule, rotation=<rot 0 0 90>}]
 dyn_shirt_fr_mid_01_r [{shapeType=Capsule, length=6.25, radius=5, shapeName=dyn_shirt_fr_mid_01_r_capsule, center=<1 0 4.6907>, rotation=<rot 90 -165.9638 104.0362>}]
 dyn_shirt_fr_mid_01_r1 [{
   shapeType=Capsule
   length=6.25
   radius=5
   shapeName=dyn_shirt_fr_mid_01_r1_capsule
   center=<1 0 4.6907>
   rotation=<rot 90 -165.9638 104.0362>
  }]
 dyn_shirt_fr_mid_01_r2 [{
   shapeType=Capsule
   length=6.25
   radius=5
   shapeName=dyn_shirt_fr_mid_01_r2_capsule
   center=<1 0 4.6907>
   rotation=<rot 90 -165.9638 104.0362>
  }]
 dyn_shirt_fr_mid_01_r_MSTR1 [{shapeType=Capsule, length=1, radius=1, shapeName=dyn_shirt_fr_mid_01_r_MSTR1_capsule, rotation=<rot 0 0 90>}]
 dyn_suit_bk_lt [{shapeType=Capsule, length=10, radius=5, shapeName=dyn_suit_bk_lt_capsule, center=<1 0 4.6907>, rotation=<rot 90 -165.9638 104.0362>}]
 dyn_suit_bk_lt1 [{shapeType=Capsule, length=10, radius=5, shapeName=dyn_suit_bk_lt1_capsule, center=<1 0 4.6907>, rotation=<rot 90 -165.9638 104.0362>}]
 dyn_suit_bk_lt2 [{shapeType=Capsule, length=10, radius=5, shapeName=dyn_suit_bk_lt2_capsule, center=<1 0 4.6907>, rotation=<rot 90 -165.9638 104.0362>}]
 dyn_suit_bk_lt_MSTR [{shapeType=Capsule, length=1, radius=1, shapeName=dyn_suit_bk_lt_MSTR_capsule, rotation=<rot 0 0 90>}]
 dyn_suit_bk_rt [{shapeType=Capsule, length=10, radius=5, shapeName=dyn_suit_bk_rt_capsule, center=<1 0 4.6907>, rotation=<rot 90 -165.9638 104.0362>}]
 dyn_suit_bk_rt1 [{shapeType=Capsule, length=10, radius=5, shapeName=dyn_suit_bk_rt1_capsule, center=<1 0 4.6907>, rotation=<rot 90 -165.9638 104.0362>}]
 dyn_suit_bk_rt2 [{shapeType=Capsule, length=10, radius=5, shapeName=dyn_suit_bk_rt2_capsule, center=<1 0 4.6907>, rotation=<rot 90 -165.9638 104.0362>}]
 dyn_suit_bk_rt_MSTR [{shapeType=Capsule, length=1, radius=1, shapeName=dyn_suit_bk_rt_MSTR_capsule, rotation=<rot 0 0 90>}]
 pelvis [
  {shapeType=Capsule, length=8, radius=9, center=<0 -0.404 0>}
  {shapeType=Capsule, length=10, radius=9.5, shapeName=pelvis_capsule, center=<-7.6245 -1.1468 -0.0001>}
 ]
 thigh_l [{shapeType=Capsule, length=20.932, radius=7.8249, center=<-19.8231 -1.0569 1.7997>, rotation=<rot 0 92.9842 -89.7376>}]
 thigh_r [{shapeType=Capsule, length=20.8001, radius=7.7525, center=<16.0244 0.8324 -1.5763>, rotation=<rot 0 92.4152 -88.9477>}]
 thigh_twist_01_r [{shapeType=Capsule, length=13, radius=8, shapeName=thigh_twist_01_r_capsule, center=<-15.1873 0.2102 -2.6652>, rotation=<rot 90 0 0>}]
}
constraints [
 {bone1Name=thigh_l, bone2Name=pelvis, swing1LimitDegrees=45, swing2LimitDegrees=45, twistLimitDegrees=45}
 {bone1Name=thigh_r, bone2Name=pelvis, swing1LimitDegrees=45, swing2LimitDegrees=45, twistLimitDegrees=45}
 {
  bone1Name=dyn_suit_bk_rt_MSTR
  bone2Name=pelvis
  swing1LimitDegrees=45
  swing1Motion=Free
  swing2LimitDegrees=45
  swing2Motion=Free
  twistLimitDegrees=45
  twistMotion=Free
 }
 {
  bone1Name=dyn_suit_bk_lt_MSTR
  bone2Name=pelvis
  swing1LimitDegrees=45
  swing1Motion=Free
  swing2LimitDegrees=45
  swing2Motion=Free
  twistLimitDegrees=45
  twistMotion=Free
 }
 {
  bone1Name=dyn_shirt_fr_01_r_MSTR
  bone2Name=pelvis
  swing1LimitDegrees=45
  swing1Motion=Free
  swing2LimitDegrees=45
  swing2Motion=Free
  twistLimitDegrees=45
  twistMotion=Free
 }
 {
  bone1Name=dyn_shirt_01_r_MSTR
  bone2Name=pelvis
  swing1LimitDegrees=45
  swing1Motion=Free
  swing2LimitDegrees=45
  swing2Motion=Free
  twistLimitDegrees=45
  twistMotion=Free
 }
 {
  bone1Name=dyn_shirt_fr_01_l_MSTR
  bone2Name=pelvis
  swing1LimitDegrees=45
  swing1Motion=Free
  swing2LimitDegrees=45
  swing2Motion=Free
  twistLimitDegrees=45
  twistMotion=Free
 }
 {
  bone1Name=dyn_shirt_01_l_MSTR
  bone2Name=pelvis
  swing1LimitDegrees=45
  swing1Motion=Free
  swing2LimitDegrees=45
  swing2Motion=Free
  twistLimitDegrees=45
  twistMotion=Free
 }
 {bone1Name=dyn_shirt_fr_mid_01_l_MSTR, bone2Name=pelvis, swing1LimitDegrees=45, swing2LimitDegrees=45, twistLimitDegrees=45}
 {
  bone1Name=dyn_shirt_fr_mid_01_r_MSTR1
  bone2Name=pelvis
  swing1LimitDegrees=45
  swing1Motion=Free
  swing2LimitDegrees=45
  swing2Motion=Free
  twistLimitDegrees=45
  twistMotion=Free
 }
 {
  bone1Name=dyn_suit_bk_rt
  bone2Name=dyn_suit_bk_rt_MSTR
  swing1LimitDegrees=10
  swing1Motion=Locked
  swing2LimitDegrees=15
  swing2Motion=Locked
  twistLimitDegrees=1
  twistMotion=Locked
 }
 {
  bone1Name=dyn_suit_bk_lt
  bone2Name=dyn_suit_bk_lt_MSTR
  swing1LimitDegrees=10
  swing1Motion=Locked
  swing2LimitDegrees=15
  swing2Motion=Locked
  twistLimitDegrees=1
  twistMotion=Locked
 }
 {bone1Name=dyn_shirt_fr_01_r, bone2Name=dyn_shirt_fr_01_r_MSTR, swing1LimitDegrees=10, swing2LimitDegrees=30, twistLimitDegrees=1}
 {bone1Name=dyn_shirt_01_r, bone2Name=dyn_shirt_01_r_MSTR, swing1LimitDegrees=10, swing2LimitDegrees=30, twistLimitDegrees=1}
 {bone1Name=dyn_shirt_fr_01_l, bone2Name=dyn_shirt_fr_01_l_MSTR, swing1LimitDegrees=10, swing2LimitDegrees=30, twistLimitDegrees=1}
 {bone1Name=dyn_shirt_01_l, bone2Name=dyn_shirt_01_l_MSTR, swing1LimitDegrees=10, swing2LimitDegrees=30, twistLimitDegrees=1}
 {bone1Name=dyn_shirt_fr_mid_01_l, bone2Name=dyn_shirt_fr_mid_01_l_MSTR, swing1LimitDegrees=10, swing2LimitDegrees=30, twistLimitDegrees=1}
 {bone1Name=dyn_shirt_fr_mid_01_r, bone2Name=dyn_shirt_fr_mid_01_r_MSTR1, swing1LimitDegrees=10, swing2LimitDegrees=30, twistLimitDegrees=1}
 {bone1Name=dyn_suit_bk_rt1, bone2Name=dyn_suit_bk_rt, swing1LimitDegrees=10, swing2LimitDegrees=30, twistLimitDegrees=20}
 {bone1Name=dyn_suit_bk_rt2, bone2Name=dyn_suit_bk_rt1, swing1LimitDegrees=10, swing2LimitDegrees=30, twistLimitDegrees=20}
 {bone1Name=dyn_suit_bk_lt1, bone2Name=dyn_suit_bk_lt, swing1LimitDegrees=10, swing2LimitDegrees=30, twistLimitDegrees=20}
 {bone1Name=dyn_suit_bk_lt2, bone2Name=dyn_suit_bk_lt1, swing1LimitDegrees=10, swing2LimitDegrees=30, twistLimitDegrees=20}
 {bone1Name=dyn_shirt_fr_01_r1, bone2Name=dyn_shirt_fr_01_r, swing1LimitDegrees=10, swing2LimitDegrees=50, twistLimitDegrees=30}
 {bone1Name=dyn_shirt_fr_01_r2, bone2Name=dyn_shirt_fr_01_r1, swing1LimitDegrees=10, swing2LimitDegrees=50, twistLimitDegrees=30}
 {bone1Name=dyn_shirt_01_r1, bone2Name=dyn_shirt_01_r, swing1LimitDegrees=30, swing2LimitDegrees=30, twistLimitDegrees=40}
 {bone1Name=dyn_shirt_01_r2, bone2Name=dyn_shirt_01_r1, swing1LimitDegrees=30, swing2LimitDegrees=30, twistLimitDegrees=40}
 {bone1Name=dyn_shirt_fr_01_l1, bone2Name=dyn_shirt_fr_01_l, swing1LimitDegrees=10, swing2LimitDegrees=50, twistLimitDegrees=20}
 {bone1Name=dyn_shirt_fr_01_l2, bone2Name=dyn_shirt_fr_01_l1, swing1LimitDegrees=10, swing2LimitDegrees=20, twistLimitDegrees=15}
 {bone1Name=dyn_shirt_01_l1, bone2Name=dyn_shirt_01_l, swing1LimitDegrees=10, swing2LimitDegrees=30, twistLimitDegrees=1}
 {bone1Name=dyn_shirt_01_l2, bone2Name=dyn_shirt_01_l1, swing1LimitDegrees=10, swing2LimitDegrees=30, twistLimitDegrees=1}
 {bone1Name=dyn_shirt_fr_mid_01_l1, bone2Name=dyn_shirt_fr_mid_01_l, swing1LimitDegrees=10, swing2LimitDegrees=30, twistLimitDegrees=5}
 {bone1Name=dyn_shirt_fr_mid_01_l2, bone2Name=dyn_shirt_fr_mid_01_l1, swing1LimitDegrees=10, swing2LimitDegrees=20, twistLimitDegrees=15}
 {bone1Name=dyn_shirt_fr_mid_01_r1, bone2Name=dyn_shirt_fr_mid_01_r, swing1LimitDegrees=10, swing2LimitDegrees=30, twistLimitDegrees=1}
 {bone1Name=dyn_shirt_fr_mid_01_r2, bone2Name=dyn_shirt_fr_mid_01_r1, swing1LimitDegrees=10, swing2LimitDegrees=30, twistLimitDegrees=1}
 {
  bone1Name=thigh_twist_01_r
  bone2Name=thigh_r
  swing1LimitDegrees=45
  swing1Motion=Free
  swing2LimitDegrees=45
  swing2Motion=Free
  twistLimitDegrees=45
  twistMotion=Free
 }
]
