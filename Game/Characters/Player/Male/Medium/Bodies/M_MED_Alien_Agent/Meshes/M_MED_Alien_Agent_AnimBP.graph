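# Post-process style Anim Blueprint for the male Alien Agent body that keeps the shirt bones clear of the moving legs.
# It copies the parent mesh pose, then runs pelvis-space RigidBody physics, eased during skydiving or parachuting.
# Thigh angles relative to the pelvis drive alphas for additive shirt-bone offsets and scaling.
# Crouching blends in calf-targeted Look At corrections, plus an extra offset when idle.
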AnimBlueprint /Game/Characters/Player/Male/Medium/Bodies/M_MED_Alien_Agent/Meshes/M_MED_Alien_Agent_AnimBP
target_skeleton=&/M_MED_Alien_Agent_Skeleton
graphs [
 AnimGraph {nodes [
   n1 = Look At - Bone: dyn_shirt_fr_01_l(ComponentPose=Local To Component(LocalPose=Use cached pose 'BaseWithRigidPose'()), Alpha=1) {
    alphaBoolBlend {blendOption=Linear}
    boneToModify {boneName=dyn_shirt_fr_01_l}
    lookAtLocation {x=0}
    lookAtTarget {boneReference {boneName=calf_l}}
    lookAt_Axis {axis=<1 0>}
    lookUp_Axis {axis {x=0, z=1}}
   }
   n2 = Look At - Bone: dyn_shirt_fr_mid_01_l(ComponentPose=@n1, Alpha=0.8) {
    alphaBoolBlend {blendOption=Linear}
    boneToModify {boneName=dyn_shirt_fr_mid_01_l}
    lookAtLocation {x=0}
    lookAtTarget {boneReference {boneName=calf_l}}
    lookAt_Axis {axis=<1 0>}
    lookUp_Axis {axis {x=0, z=1}}
   }
   n3 = Look At - Bone: dyn_shirt_fr_mid_01_l1(ComponentPose=@n2, Alpha=0.8) {
    alphaBoolBlend {blendOption=Linear}
    boneToModify {boneName=dyn_shirt_fr_mid_01_l1}
    lookAtLocation {x=0}
    lookAtTarget {boneReference {boneName=calf_l}}
    lookAt_Axis {axis=<1 0>}
    lookUp_Axis {axis {x=0, z=1}}
   }
   n4 = Look At - Bone: dyn_shirt_01_l(ComponentPose=@n3, Alpha=0.5) {
    alphaBoolBlend {blendOption=Linear}
    boneToModify {boneName=dyn_shirt_01_l}
    lookAtLocation {x=0}
    lookAtTarget {boneReference {boneName=calf_l}}
    lookAt_Axis {axis=<1 0>}
    lookUp_Axis {axis {x=0, z=1}}
   }
   n5 = Look At - Bone: dyn_shirt_fr_01_r(ComponentPose=@n4, Alpha=1) {
    alphaBoolBlend {blendOption=Linear}
    boneToModify {boneName=dyn_shirt_fr_01_r}
    lookAtLocation {x=0}
    lookAtTarget {boneReference {boneName=calf_r}}
    lookAt_Axis {axis=<1 0>}
    lookUp_Axis {axis {x=0, z=1}}
   }
   n6 = Get Left Leg() -> (Left Leg:double)
   n7 = "Transform (Modify) Bone - Bone: dyn_shirt_fr_01_l_MSTR"(ComponentPose=@n5, Alpha=@n6, Translation="-1.279,-0.000,0.000", Rotation="5.000,-0.001,-0.000", Scale="1,1,1") {
    rotationMode=BMM_Additive
    rotationSpace=BCS_BoneSpace
    scaleMode=BMM_Additive
    scaleSpace=BCS_BoneSpace
    translationMode=BMM_Additive
    translationSpace=BCS_BoneSpace
    alphaBoolBlend {blendOption=Linear}
    alphaScaleBiasClamp {bInterpResult=true, interpSpeedDecreasing=50, interpSpeedIncreasing=50}
    boneToModify {boneName=dyn_shirt_fr_01_l_MSTR}
    rotation=<rot 5.0001 -0.001 0>
    translation {x=-1.279}
   }
   n8 = "Transform (Modify) Bone - Bone: dyn_shirt_fr_mid_01_l_MSTR"(ComponentPose=@n7, Alpha=@n6, Translation="-2.364,-0.021,0.008", Rotation="9.994,0.000,-5.000", Scale="1,1,1") {
    rotationMode=BMM_Additive
    rotationSpace=BCS_BoneSpace
    scaleMode=BMM_Additive
    scaleSpace=BCS_BoneSpace
    translationMode=BMM_Additive
    translationSpace=BCS_BoneSpace
    alphaBoolBlend {blendOption=Linear}
    alphaScaleBiasClamp {bInterpResult=true, interpSpeedDecreasing=50, interpSpeedIncreasing=50}
    boneToModify {boneName=dyn_shirt_fr_mid_01_l_MSTR}
    rotation=<rot 9.9941 0 -5>
    translation=<-2.364 -0.021 0.008>
   }
   n9 = Get Right Leg_Crouch() -> (Right Leg_Crouch:double)
   n10 = "Transform (Modify) Bone - Bone: dyn_shirt_fr_01_l_MSTR"(ComponentPose=@n8, Alpha=@n9, Translation="-1.810,-0.000,-0.376", Rotation="24.804,1.263,5.399", Scale="1,1,1") {
    rotationMode=BMM_Additive
    rotationSpace=BCS_BoneSpace
    scaleMode=BMM_Additive
    scaleSpace=BCS_BoneSpace
    translationMode=BMM_Additive
    translationSpace=BCS_BoneSpace
    alphaBoolBlend {blendOption=Linear}
    alphaScaleBiasClamp {bInterpResult=true}
    boneToModify {boneName=dyn_shirt_fr_01_l_MSTR}
    rotation=<rot 24.8041 1.263 5.399>
    translation {x=-1.81, z=-0.376}
   }
   n11 = "Transform (Modify) Bone - Bone: dyn_shirt_fr_01_r_MSTR"(ComponentPose=@n10, Alpha=@n9, Translation="0,0,0", Rotation="0,0,0", Scale="1,1,1") {
    rotationMode=BMM_Additive
    rotationSpace=BCS_BoneSpace
    scaleMode=BMM_Additive
    scaleSpace=BCS_BoneSpace
    translationMode=BMM_Additive
    translationSpace=BCS_BoneSpace
    alphaBoolBlend {blendOption=Linear}
    alphaScaleBiasClamp {bInterpResult=true}
    boneToModify {boneName=dyn_shirt_fr_01_r_MSTR}
   }
   n12 = Component To Local(ComponentPose=@n11)
   n13 = "Transform (Modify) Bone - Bone: dyn_shirt_fr_01_l_MSTR"(
    ComponentPose=Local To Component(LocalPose=Use cached pose 'BaseWithRigidPose'())
    Alpha=@n6
    Translation="0.151,-3.175,-2.202"
    Rotation="31.690,-1.906,-17.209"
    Scale="0.5,0.5,0.5"
   ) {
    rotationMode=BMM_Additive
    rotationSpace=BCS_BoneSpace
    scaleMode=BMM_Additive
    scaleSpace=BCS_BoneSpace
    translationMode=BMM_Additive
    translationSpace=BCS_BoneSpace
    alphaBoolBlend {blendOption=Linear}
    alphaScaleBiasClamp {bInterpResult=true, interpSpeedDecreasing=50, interpSpeedIncreasing=50}
    boneToModify {boneName=dyn_shirt_fr_01_l_MSTR}
    rotation=<rot 31.69 -1.906 -17.209>
    scale=<0.5 0.5 0.5>
    translation=<0.151 -3.175 -2.202>
   }
   n14 = "Transform (Modify) Bone - Bone: dyn_shirt_fr_mid_01_l_MSTR"(ComponentPose=@n13, Alpha=@n6, Translation="-3,0,0", Rotation="10,0,0", Scale="0.8,0.8,0.8") {
    rotationMode=BMM_Additive
    rotationSpace=BCS_BoneSpace
    scaleMode=BMM_Additive
    scaleSpace=BCS_BoneSpace
    translationMode=BMM_Additive
    translationSpace=BCS_BoneSpace
    alphaBoolBlend {blendOption=Linear}
    alphaScaleBiasClamp {bInterpResult=true, interpSpeedDecreasing=50, interpSpeedIncreasing=50}
    boneToModify {boneName=dyn_shirt_fr_mid_01_l_MSTR}
    rotation {pitch=10}
    scale=<0.8 0.8 0.8>
    translation {x=-3}
   }
   n15 = "Transform (Modify) Bone - Bone: dyn_shirt_fr_mid_01_l2"(ComponentPose=@n14, Alpha=@n6, Translation="0,0,0", Rotation="48.551,-10.374,-4.113", Scale="1,1,1") {
    rotationMode=BMM_Additive
    rotationSpace=BCS_BoneSpace
    scaleMode=BMM_Additive
    scaleSpace=BCS_BoneSpace
    translationMode=BMM_Additive
    translationSpace=BCS_BoneSpace
    alphaBoolBlend {blendOption=Linear}
    alphaScaleBiasClamp {bInterpResult=true, interpSpeedDecreasing=50, interpSpeedIncreasing=50}
    boneToModify {boneName=dyn_shirt_fr_mid_01_l2}
    rotation=<rot 48.551 -10.374 -4.113>
   }
   n16 = "Transform (Modify) Bone - Bone: dyn_shirt_01_l_MSTR"(ComponentPose=@n15, Alpha=@n6, Translation="0,-5,0", Rotation="3.757,-0.004,0.012", Scale="0.8,0.8,0.8") {
    rotationMode=BMM_Additive
    rotationSpace=BCS_BoneSpace
    scaleMode=BMM_Additive
    scaleSpace=BCS_BoneSpace
    translationMode=BMM_Additive
    translationSpace=BCS_BoneSpace
    alphaBoolBlend {blendOption=Linear}
    alphaScaleBiasClamp {bInterpResult=true, interpSpeedDecreasing=50, interpSpeedIncreasing=50}
    boneToModify {boneName=dyn_shirt_01_l_MSTR}
    rotation=<rot 3.757 -0.004 0.012>
    scale=<0.8 0.8 0.8>
    translation {y=-5}
   }
   n17 = Component To Local(ComponentPose=@n16)
   n18 = Get bIsCrouching()
   n19 = Get Right Leg() -> (Right Leg:double)
   n20 = "Transform (Modify) Bone - Bone: dyn_shirt_01_r_MSTR"(
    ComponentPose=Local To Component(LocalPose=Blend Poses by bool(BlendPose_0=@n12, BlendPose_1=@n17, bActiveValue=@n18) {blendType=Linear})
    Alpha=@n19
    Translation="0,-5,0"
    Rotation="15,0,0"
    Scale="0.8,0.8,0.8"
   ) {
    rotationMode=BMM_Additive
    rotationSpace=BCS_BoneSpace
    scaleMode=BMM_Additive
    scaleSpace=BCS_BoneSpace
    translationMode=BMM_Additive
    translationSpace=BCS_BoneSpace
    alphaBoolBlend {blendOption=Linear}
    alphaScaleBiasClamp {bInterpResult=true}
    boneToModify {boneName=dyn_shirt_01_r_MSTR}
    rotation {pitch=15}
    scale=<0.8 0.8 0.8>
    translation {y=-5}
   }
   n21 = "Transform (Modify) Bone - Bone: dyn_shirt_fr_01_r_MSTR"(ComponentPose=@n20, Alpha=@n19, Translation="0,0,0", Rotation="35,10,0", Scale="0.5,0.5,0.5") {
    rotationMode=BMM_Additive
    rotationSpace=BCS_BoneSpace
    scaleMode=BMM_Additive
    scaleSpace=BCS_BoneSpace
    translationMode=BMM_Additive
    translationSpace=BCS_BoneSpace
    alphaBoolBlend {blendOption=Linear}
    alphaScaleBiasClamp {bInterpResult=true}
    boneToModify {boneName=dyn_shirt_fr_01_r_MSTR}
    rotation {pitch=35, yaw=10}
    scale=<0.5 0.5 0.5>
   }
   n22 = "Transform (Modify) Bone - Bone: dyn_shirt_fr_mid_01_r_MSTR1"(ComponentPose=@n21, Alpha=@n19, Translation="0,0,0", Rotation="50,0,0", Scale="0.8,0.8,0.8") {
    rotationMode=BMM_Additive
    rotationSpace=BCS_BoneSpace
    scaleMode=BMM_Additive
    scaleSpace=BCS_BoneSpace
    translationMode=BMM_Additive
    translationSpace=BCS_BoneSpace
    alphaBoolBlend {blendOption=Linear}
    alphaScaleBiasClamp {bInterpResult=true}
    boneToModify {boneName=dyn_shirt_fr_mid_01_r_MSTR1}
    rotation {pitch=50}
    scale=<0.8 0.8 0.8>
   }
   n23 = Get Right Leg_Crouch()
   n24 = "Transform (Modify) Bone - Bone: dyn_shirt_01_r_MSTR"(ComponentPose=@n22, Alpha=@n23, Translation="-0.001,-5.000,8.324", Rotation="0,0,0", Scale="0.8,0.8,0.8") {
    rotationMode=BMM_Additive
    rotationSpace=BCS_BoneSpace
    scaleMode=BMM_Additive
    scaleSpace=BCS_BoneSpace
    translationMode=BMM_Additive
    translationSpace=BCS_BoneSpace
    alphaBoolBlend {blendOption=Linear}
    alphaScaleBiasClamp {bInterpResult=true, interpSpeedIncreasing=25}
    boneToModify {boneName=dyn_shirt_01_r_MSTR}
    scale=<0.8 0.8 0.8>
    translation=<-0.001 -5 8.324>
   }
   n25 = "Transform (Modify) Bone - Bone: dyn_shirt_fr_01_r_MSTR"(ComponentPose=@n24, Alpha=@n23, Translation="-1.627,0.860,-0.923", Rotation="10.000,10.010,-0.329", Scale="0.8,0.8,0.8") {
    rotationMode=BMM_Additive
    rotationSpace=BCS_BoneSpace
    scaleMode=BMM_Additive
    scaleSpace=BCS_BoneSpace
    translationMode=BMM_Additive
    translationSpace=BCS_BoneSpace
    alphaBoolBlend {blendOption=Linear}
    alphaScaleBiasClamp {bInterpResult=true, interpSpeedIncreasing=25}
    boneToModify {boneName=dyn_shirt_fr_01_r_MSTR}
    rotation=<rot 10.0001 10.01 -0.329>
    scale=<0.8 0.8 0.8>
    translation=<-1.627 0.86 -0.923>
   }
   n26 = "Transform (Modify) Bone - Bone: dyn_shirt_fr_mid_01_r_MSTR1"(ComponentPose=@n25, Alpha=@n23, Translation="0,0,0", Rotation="30,0,0", Scale="0.8,0.8,0.8") {
    rotationMode=BMM_Additive
    rotationSpace=BCS_BoneSpace
    scaleMode=BMM_Additive
    scaleSpace=BCS_BoneSpace
    translationMode=BMM_Additive
    translationSpace=BCS_BoneSpace
    alphaBoolBlend {blendOption=Linear}
    alphaScaleBiasClamp {bInterpResult=true, interpSpeedIncreasing=25}
    boneToModify {boneName=dyn_shirt_fr_mid_01_r_MSTR1}
    rotation {pitch=30}
    scale=<0.8 0.8 0.8>
   }
   n27 = Get CrouchIdle_alpha() -> (CrouchIdle_alpha:double)
   n28 = "Transform (Modify) Bone - Bone: dyn_shirt_fr_01_l_MSTR"(ComponentPose=@n26, Alpha=@n27, Translation="1.353,-0.381,-0.017", Rotation="2.875,0.000,-0.000", Scale="1,1,1") {
    rotationMode=BMM_Additive
    rotationSpace=BCS_BoneSpace
    scaleMode=BMM_Additive
    scaleSpace=BCS_BoneSpace
    translationMode=BMM_Additive
    translationSpace=BCS_BoneSpace
    alphaBoolBlend {blendOption=Linear}
    alphaScaleBiasClamp {bInterpResult=true, interpSpeedDecreasing=50, interpSpeedIncreasing=50}
    boneToModify {boneName=dyn_shirt_fr_01_l_MSTR}
    rotation=<rot 2.8751 0 0>
    translation=<1.353 -0.381 -0.017>
   }
   n29 = "Transform (Modify) Bone - Bone: dyn_shirt_fr_mid_01_l_MSTR"(ComponentPose=@n28, Alpha=@n27, Translation="1.972,0.000,-0.000", Rotation="0,0,0", Scale="1,1,1") {
    rotationMode=BMM_Additive
    rotationSpace=BCS_BoneSpace
    scaleMode=BMM_Additive
    scaleSpace=BCS_BoneSpace
    translationMode=BMM_Additive
    translationSpace=BCS_BoneSpace
    alphaBoolBlend {blendOption=Linear}
    alphaScaleBiasClamp {bInterpResult=true, interpSpeedDecreasing=50, interpSpeedIncreasing=50}
    boneToModify {boneName=dyn_shirt_fr_mid_01_l_MSTR}
    translation=<1.9716 0 0>
   }
   Output Pose(Result=Component To Local(ComponentPose=@n29)) {layerGroup=None}
   n30 = RigidBody(
    ComponentPose=Local To Component(LocalPose=Copy Pose From Mesh() {bUseAttachedParent=true})
    Alpha=Get RigidBodyAlpha() -> (RigidBodyAlpha:float)
   ) {
    bClampLinearTranslationLimitToRefPose=true
    bOverrideWorldGravity=true
    overridePhysicsAsset=&/M_MED_Alien_Agent_V2_Physics
    simulationSpace=BaseBoneSpace
    alphaBoolBlend {blendOption=Linear}
    baseBoneRef {boneName=pelvis}
    componentAppliedLinearAccClamp=<3000 3000 3000>
    componentLinearAccScale=<0.45 0.5 0.8>
    componentLinearVelScale=<0.45 0.5 0.8>
    overrideWorldGravity {z=-980}
    simSpaceSettings {maxAngularAcceleration=3000, maxAngularVelocity=3000, maxLinearAcceleration=3000, maxLinearVelocity=3000, worldAlpha=0.344}
   }
   Save cached pose 'BaseWithRigidPose'(Pose=Component To Local(ComponentPose=@n30)) {cacheName=BaseWithRigidPose}
  ]}
 EventGraph {nodes [
   Event BlueprintUpdateAnimation()
   > n1 = Sequence!(execute) -> (then_0, then_1, then_2:Exec, then_3:Exec)
   > Get Dynamic Material Instance(SkeletalMeshComponent=GetPartSkeletalMeshComponent())
   > GetScalarParameterValue!(execute, Collection:Material Parameter Collection Object Reference, ParameterName:Name=None) -> (then, ReturnValue:float) {is=title K2Node_CallMaterialParameterCollectionFunction}
   Set Left Leg(
    execute=@n1.then_1
    Left Leg:double=GetAlphaFromReferenceTransformDelta(SkeletalMeshComponent=Get Skel Mesh(), BoneName=thigh_l, BaseName=pelvis, InputRange=(X=-23,Y=-102), TwistAxis="1,0,0")
   ) -> (Output_Get:double)
   n2 = Branch(execute=@n1.then_2, Condition=Get bIsCrouching())
   > n3 = Set Right Leg_Crouch(Right Leg_Crouch:double=GetAlphaFromReferenceTransformDelta(SkeletalMeshComponent=Get Skel Mesh(), BoneName=thigh_r, BaseName=pelvis, InputRange=(X=-24,Y=-104), TwistAxis="1,0,0")) -> (Output_Get:double)
   n4 = NOT Boolean(A=Get bIsMoving2D())
   > Set CrouchIdle_alpha(CrouchIdle_alpha:double=float * float(A=float - float(A=1, B=@n3), B="To Float (Boolean)"(InBool=@n4)) {is=title K2Node_PromotableOperator}) -> (Output_Get:double)
   Set Right Leg_Crouch(execute=@n2.else, Right Leg_Crouch=0)
   > Set CrouchIdle_alpha(CrouchIdle_alpha=0)
   > Set Right Leg(Right Leg:double=GetAlphaFromReferenceTransformDelta(SkeletalMeshComponent=Get Skel Mesh(), BoneName=thigh_r, BaseName=pelvis, InputRange=(X=-19,Y=-100), TwistAxis="1,0,0")) -> (Output_Get:double)
   n5 = Get bIsSkyDiving()
   n6 = Get bIsParachuteOpened()
   Set RigidBodyAlpha(execute=@n1.then_3, RigidBodyAlpha:float=SelectFloat(A=0.7, B=1, bPickA=OR Boolean(A=@n5, B=@n6))) -> (Output_Get:float)
   Event BlueprintInitializeAnimation()
   > Set Skel Mesh(Skel Mesh=GetOwnerSkeletalMeshComponent())
  ]}
]
variables [
 CrouchIdle_alpha {type=number}
 Left Leg {type=number}
 Right Leg {type=number}
 Right Leg_Crouch {type=number}
 RigidBodyAlpha {type=number, default=1}
 Skel Mesh {type=SkeletalMeshComponent}
]
